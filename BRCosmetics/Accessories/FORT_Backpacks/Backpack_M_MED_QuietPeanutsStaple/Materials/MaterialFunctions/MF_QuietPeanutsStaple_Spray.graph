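# Material function for the QuietPeanutsStaple backpack that layers two optional sticker "sprays" (top and bottom) onto incoming material attributes.
# Each spray is positioned and center-scaled with clamped UVs, and its squared color replaces base color under a mask.
# Offset parameters adjust specular and roughness, though both blend from the input specular.
# A ShellWear mask, built from packed-mask, scratch and noise textures, is also output as MaskDebug.
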MaterialFunction /BRCosmetics/Accessories/FORT_Backpacks/Backpack_M_MED_QuietPeanutsStaple/Materials/MaterialFunctions/MF_QuietPeanutsStaple_Spray
graph {nodes [
  n1 = GetMaterialAttributes(None=FunctionInput() {inputName=MA, inputType=FunctionInput_MaterialAttributes}) {attributeGetTypes [#1, #2, #3]}
  n2 = TextureCoordinate()
  n3 = VectorParameter() {group=QuietPeanutsStaple_Spray, parameterName=TopSpray_Scale, sortPriority=12, defaultValue=<rgba 0.2 0.2 0 0>}
  n4 = VectorParameter() {group=QuietPeanutsStaple_Spray, parameterName=TopSpray_Position, sortPriority=11, defaultValue=<rgba 0.5 0.5 0 1>}
  n5 = AppendVector(A=LinearInterpolate(Alpha=@n4.R) {constA=0.6, constB=-0.18}, B=LinearInterpolate(Alpha=@n4.G) {constA=-0.55, constB=0.28})
  n6 = ComponentMask(None=@n3.RGB) {g=true, r=true}
  n7 = TextureSampleParameter2D(UVs=Add(
    A=MaterialFunctionCall(UVs=@n2, Texture Scale=@n3.RGB) {materialFunction=&ScaleUVsByCenter}.UVs
    B=Divide(A=@n5, B=@n6)
   )) {
   group=QuietPeanutsStaple_Spray
   parameterName=TopSpray_Texture
   samplerSource=SSM_Clamp_WorldGroupSettings
   sortPriority=10
   texture=/BRCosmetics/2dAssets/Emoji/Season14/Elastic/T-Emote-Icons-Season14-S14-Sticker-Elastic-butterfly
  }
  n8 = Multiply(A=@n7.RGB, B=@n7.RGB)
  n9 = Constant()
  n10 = TextureCoordinate()
  n11 = VectorParameter() {group=QuietPeanutsStaple_Spray, parameterName=BottomSpray_Scale, sortPriority=16, defaultValue=<rgba 0.2 0.2 0 1>}
  n12 = VectorParameter() {group=QuietPeanutsStaple_Spray, parameterName=BottomSpray_Position, sortPriority=15, defaultValue=<rgba 0.5 0.5 0 1>}
  n13 = AppendVector(A=LinearInterpolate(Alpha=@n12.R) {constA=0.6, constB=-0.18}, B=LinearInterpolate(Alpha=@n12.G) {constA=-0.55, constB=0.28})
  n14 = ComponentMask(None=@n11.RGB) {g=true, r=true}
  n15 = TextureSampleParameter2D(UVs=Add(
    A=MaterialFunctionCall(UVs=@n10, Texture Scale=@n11.RGB) {materialFunction=&ScaleUVsByCenter}.UVs
    B=Divide(A=@n13, B=@n14)
   )) {
   group=QuietPeanutsStaple_Spray
   parameterName=BottomSpray_Texture
   samplerSource=SSM_Clamp_WorldGroupSettings
   sortPriority=14
   texture=&T-Emote-Icons-Season14-S14-Sticker-Elastic-butterfly
  }
  n16 = Multiply(A=@n15.RGB, B=@n15.RGB)
  n17 = Constant()
  n18 = Constant()
  n19 = Subtract(
   A=Add(
    A=StaticSwitchParameter(True=@n8, False=@n9) {group=QuietPeanutsStaple_Spray, parameterName=UseTopSpray, sortPriority=9}
    B=StaticSwitchParameter(True=@n16, False=@n17) {group=QuietPeanutsStaple_Spray, parameterName=UseBottomSpray, sortPriority=13}
   )
   B=@n18
  )
  n20 = NamedRerouteUsage()
  n21 = AppendVector(A=@n1.Specular, B=@n1.Specular)
  n22 = AppendVector(
   A=Saturate(None=ScalarParameter() {group=QuietPeanutsStaple_Spray, parameterName=Spray_SpecularOffset, sliderMax=1, sortPriority=31})
   B=Saturate(None=ScalarParameter() {group=QuietPeanutsStaple_Spray, parameterName=Spray_RoughnessOffset, sliderMax=1})
  )
  n23 = LinearInterpolate(
   A=@n21
   B=LinearInterpolate(
    A=@n22
    B=Multiply(A=@n22, B=@n21)
    Alpha=ScalarParameter() {group=QuietPeanutsStaple_Spray, parameterName="BlendRough&&Spec", sliderMax=1, sortPriority=30}
   )
   Alpha=@n20
  )
  FunctionOutput(None=SetMaterialAttributes!(
    MaterialAttributes=@n1.MaterialAttributes
    BaseColor=LinearInterpolate(A=@n1.BaseColor, B=@n19, Alpha=@n20)
    Specular=ComponentMask(None=@n23) {r=true}
    Roughness=ComponentMask(None=@n23) {g=true}
   ) {attributeSetTypes [#3, #1, #2]}) {outputName=MA}
  n24 = TextureSample() {
   samplerType=SAMPLERTYPE_Masks
   texture=/BRCosmetics/Accessories/FORT_Backpacks/Backpack_M_MED_QuietPeanutsStaple/Textures/T_QuietPeanutsStaple_Backpack_AddtionalMasks
  }
  n25 = SphereMask(A=@n24.G, B=Constant() {r=1}) {attenuationRadius=0.35, hardnessPercent=12}
  n26 = TextureCoordinate()
  n27 = Multiply(
   A=SphereMask(A=@n24.G, B=Constant() {r=0.8}) {attenuationRadius=0.2, hardnessPercent=40}
   B=TextureSample(UVs=Multiply(A=@n26) {constB=2}) {samplerType=SAMPLERTYPE_LinearColor, texture=/Game/Global/Textures/Default/Masks_and_Gradients/T_Scratch_Mask}.R
  )
  n28 = OneMinus(None=Saturate(None=Add(A=@n25, B=@n27)))
  n29 = If(A=NamedRerouteUsage(), "A > B"=@n28, "A == B"=@n28, "A < B"=Saturate(None=Add(A=@n28) {constB=0.1})) {constB=0.1}
  n30 = TextureCoordinate()
  n31 = Constant() {r=0.2}
  n32 = Multiply(
   A=TextureSample(UVs=Multiply(A=@n30) {constB=4}) {samplerType=SAMPLERTYPE_LinearColor, texture=&T_Scratch_Mask}.R
   B=If(A=NamedRerouteUsage(), "A > B"=@n31, "A == B"=@n31, "A < B"=Multiply(A=@n31) {constB=7}) {constB=0.1}
  ) {constB=0.45}
  n33 = Multiply(A=SphereMask(A=@n24.G, B=Constant() {r=0.8}) {attenuationRadius=0.12, hardnessPercent=1}) {constB=2}
  n34 = TextureSample(UVs=Multiply(A=TextureCoordinate()) {constB=8}) {samplerType=SAMPLERTYPE_LinearGrayscale, texture=&T_SimplePerlin_Greyscale}
  n35 = Add(A=Power(Base=Multiply(A=@n34.R) {constB=0.7}) {constExponent=8}) {constB=0.7}
  n36 = Add(A=Saturate(None=Add(A=@n29, B=@n32)), B=Multiply(A=Saturate(None=@n33), B=Saturate(None=@n35)) {constB=0}) {constB=0}
  FunctionOutput(None=NamedRerouteDeclaration(None=Saturate(None=@n36)) {name=ShellWear, nodeColor=<rgba 0.7344 0 1 1>}) {outputName=MaskDebug, sortPriority=1}
  n37 = NamedRerouteUsage()
  n38 = NamedRerouteUsage()
  n39 = Add(
   A=StaticSwitchParameter(True=Multiply(A=@n7.A, B=@n37), False=@n9) {group=QuietPeanutsStaple_Spray, parameterName=UseTopSpray, sortPriority=9}
   B=StaticSwitchParameter(True=Multiply(A=@n15.A, B=@n38), False=@n17) {group=QuietPeanutsStaple_Spray, parameterName=UseBottomSpray, sortPriority=13}
  )
  If(A=@n39, "A > B"=@n16, "A == B"=@n18, "A < B"=@n18) {constB=1}
  NamedRerouteDeclaration(None=Saturate(None=@n39)) {name=SprayMask, nodeColor=<rgba 1 0 0.3516 1>}
  n40 = Desaturation(None=Add(A=@n7.RGB, B=@n15.RGB))
  NamedRerouteDeclaration(None=@n40) {name=StickerColor_Grayscale, nodeColor=<rgba 1 1 1 1>}
  Reroute(None=@n40)
 ]}
