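# Parent material for the QuietPeanutsStaple backpack cosmetic, adding selectable scarf colors on top of the 2023 character parent PreFX and PostFX functions.
# A scarf mask from the FX texture drives the color variants and the roughness and specular offsets.
# ColorIndex switches base color between a slightly hue-shifted, saturation-adjusted original and desaturated blue, red or purple tints.
# The adjusted attributes then pass through the backpack's Spray material function.
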Material /BRCosmetics/Accessories/FORT_Backpacks/Backpack_M_MED_QuietPeanutsStaple/Materials/M_QuietPeanutsStaple_Backpack_PARENT
graph {
 expressions {
  MaterialExpressionAdd_0 {
   class=MaterialExpressionAdd
   inputs [
    A {from=MaterialExpressionGetMaterialAttributes_1, output=Roughness}
    B {from=MaterialExpressionMultiply_1}
   ]
  }
  MaterialExpressionAdd_1 {
   class=MaterialExpressionAdd
   inputs [
    A {from=MaterialExpressionGetMaterialAttributes_1, output=Specular}
    B {from=MaterialExpressionMultiply_2}
   ]
  }
  MaterialExpressionConstant_1 {class=MaterialExpressionConstant, properties {r=-0.014}}
  MaterialExpressionConstant_2 {class=MaterialExpressionConstant, properties {r=-0.08}}
  MaterialExpressionDesaturation_0 {
   class=MaterialExpressionDesaturation
   inputs [
    None {from=MaterialExpressionGetMaterialAttributes_0, output=Base Color}
    Fraction {from=MaterialExpressionReroute_0, output=""}
   ]
   properties {}
  }
  MaterialExpressionDesaturation_1 {
   class=MaterialExpressionDesaturation
   inputs [
    None {from=MaterialExpressionGetMaterialAttributes_0, output=Base Color}
    Fraction {from=MaterialExpressionReroute_1, output=""}
   ]
   properties {}
  }
  MaterialExpressionDesaturation_2 {
   class=MaterialExpressionDesaturation
   inputs [
    None {from=MaterialExpressionGetMaterialAttributes_0, output=Base Color}
    Fraction {from=MaterialExpressionReroute_2, output=""}
   ]
   properties {}
  }
  MaterialExpressionDesaturation_3 {
   class=MaterialExpressionDesaturation
   inputs [
    None {from=MaterialExpressionMaterialFunctionCall_2, output=Result}
    Fraction {from=MaterialExpressionMultiply_5, output=""}
   ]
   properties {}
  }
  MaterialExpressionGetMaterialAttributes_0 {
   class=MaterialExpressionGetMaterialAttributes
   inputs [None {from=MaterialExpressionMaterialFunctionCall_1, output=Result}]
   properties {attributeGetTypes [#1]}
  }
  MaterialExpressionGetMaterialAttributes_1 {
   class=MaterialExpressionGetMaterialAttributes
   inputs [None {from=MaterialExpressionGetMaterialAttributes_0, output=MaterialAttributes}]
   properties {attributeGetTypes [#2, #3]}
  }
  MaterialExpressionLinearInterpolate_1 {
   class=MaterialExpressionLinearInterpolate
   inputs [
    A {from=MaterialExpressionDesaturation_0}
    B {from=MaterialExpressionMultiply_4}
    Alpha {from=MaterialExpressionReroute_0}
   ]
  }
  MaterialExpressionLinearInterpolate_2 {
   class=MaterialExpressionLinearInterpolate
   inputs [
    A {from=MaterialExpressionDesaturation_1}
    B {from=MaterialExpressionMultiply_6}
    Alpha {from=MaterialExpressionReroute_1}
   ]
  }
  MaterialExpressionLinearInterpolate_3 {
   class=MaterialExpressionLinearInterpolate
   inputs [
    A {from=MaterialExpressionDesaturation_2}
    B {from=MaterialExpressionMultiply_7}
    Alpha {from=MaterialExpressionReroute_2}
   ]
  }
  MaterialExpressionMaterialFunctionCall_0 {
   class=MaterialExpressionMaterialFunctionCall
   inputs [
    Overide Texture Set UVs {from=""}
    Override Tech Art Mask UVs {from=""}
    Override Basic Emissive UVs {from=""}
   ]
   properties {materialFunction=&MF_CharacterParent_2023_PreFX}
  }
  MaterialExpressionMaterialFunctionCall_1 {
   class=MaterialExpressionMaterialFunctionCall
   inputs [
    Material {from=MaterialExpressionMaterialFunctionCall_0, output=Result}
    M {from=MaterialExpressionMaterialFunctionCall_0, output=Result}
    Pre Skinned Local Position {from=MaterialExpressionMaterialFunctionCall_0, output=Result}
    Distance Blend {from=MaterialExpressionMaterialFunctionCall_0, output=Result}
    Fresnel {from=MaterialExpressionMaterialFunctionCall_0, output=Result}
   ]
   properties {materialFunction=&MF_CharacterParent_2023_PostFX}
  }
  MaterialExpressionMaterialFunctionCall_2 {
   class=MaterialExpressionMaterialFunctionCall
   inputs [
    Hue Shift Percentage {from=MaterialExpressionMultiply_3}
    Texture {from=MaterialExpressionGetMaterialAttributes_0, output=Base Color}
   ]
   properties {materialFunction=&HueShift}
  }
  MaterialExpressionMaterialFunctionCall_8 {
   class=MaterialExpressionMaterialFunctionCall
   inputs [MA {from=MaterialExpressionSetMaterialAttributes_0}]
   properties {materialFunction=/BRCosmetics/Accessories/FORT_Backpacks/Backpack_M_MED_QuietPeanutsStaple/Materials/MaterialFunctions/MF_QuietPeanutsStaple_Spray}
  }
  MaterialExpressionMultiply_0 {
   class=MaterialExpressionMultiply
   inputs [
    A {from=MaterialExpressionSubtract_0}
    B {from=""}
   ]
   properties {constB=3}
  }
  MaterialExpressionMultiply_1 {
   class=MaterialExpressionMultiply
   inputs [
    A {from=MaterialExpressionNamedRerouteUsage_2}
    B {from=MaterialExpressionScalarParameter_1}
   ]
  }
  MaterialExpressionMultiply_2 {
   class=MaterialExpressionMultiply
   inputs [
    A {from=MaterialExpressionScalarParameter_2}
    B {from=MaterialExpressionNamedRerouteUsage_3}
   ]
  }
  MaterialExpressionMultiply_3 {
   class=MaterialExpressionMultiply
   inputs [
    A {from=MaterialExpressionConstant_1}
    B {from=MaterialExpressionNamedRerouteUsage_5}
   ]
  }
  MaterialExpressionMultiply_4 {
   class=MaterialExpressionMultiply
   inputs [
    A {from=MaterialExpressionDesaturation_0}
    B {from=MaterialExpressionVectorParameter_0, output=RGB}
   ]
   properties {constB=0}
  }
  MaterialExpressionMultiply_5 {
   class=MaterialExpressionMultiply
   inputs [
    A {from=MaterialExpressionNamedRerouteUsage_5}
    B {from=MaterialExpressionConstant_2}
   ]
  }
  MaterialExpressionMultiply_6 {
   class=MaterialExpressionMultiply
   inputs [
    A {from=MaterialExpressionDesaturation_1}
    B {from=MaterialExpressionVectorParameter_1, output=RGB}
   ]
   properties {constB=0}
  }
  MaterialExpressionMultiply_7 {
   class=MaterialExpressionMultiply
   inputs [
    A {from=MaterialExpressionDesaturation_2}
    B {from=MaterialExpressionVectorParameter_2, output=RGB}
   ]
   properties {constB=0}
  }
  MaterialExpressionNamedRerouteDeclaration_0 {
   class=MaterialExpressionNamedRerouteDeclaration
   inputs [None {from=MaterialExpressionSaturate_2}]
   properties {name=ScarfMask, nodeColor=<rgba 1 0 0.3516 1>}
  }
  MaterialExpressionNamedRerouteUsage_0 {class=MaterialExpressionNamedRerouteUsage}
  MaterialExpressionNamedRerouteUsage_2 {class=MaterialExpressionNamedRerouteUsage}
  MaterialExpressionNamedRerouteUsage_3 {class=MaterialExpressionNamedRerouteUsage}
  MaterialExpressionNamedRerouteUsage_5 {class=MaterialExpressionNamedRerouteUsage}
  MaterialExpressionReroute_0 {class=MaterialExpressionReroute, inputs [None {from=MaterialExpressionNamedRerouteDeclaration_0}]}
  MaterialExpressionReroute_1 {class=MaterialExpressionReroute, inputs [None {from=MaterialExpressionNamedRerouteDeclaration_0}]}
  MaterialExpressionReroute_2 {class=MaterialExpressionReroute, inputs [None {from=MaterialExpressionNamedRerouteDeclaration_0}]}
  MaterialExpressionSaturate_0 {class=MaterialExpressionSaturate, inputs [None {from=MaterialExpressionAdd_0}]}
  MaterialExpressionSaturate_1 {class=MaterialExpressionSaturate, inputs [None {from=MaterialExpressionAdd_1}]}
  MaterialExpressionSaturate_2 {class=MaterialExpressionSaturate, inputs [None {from=MaterialExpressionMultiply_0}]}
  MaterialExpressionScalarParameter_0 {class=MaterialExpressionScalarParameter, properties {group=QuietPeanutsStaple, parameterName=ColorIndex, sliderMax=3, sortPriority=0}}
  MaterialExpressionScalarParameter_1 {
   class=MaterialExpressionScalarParameter
   properties {group=QuietPeanutsStaple, parameterName=Scarf_RoughnessOffset, sliderMax=3, sortPriority=31}
  }
  MaterialExpressionScalarParameter_2 {
   class=MaterialExpressionScalarParameter
   properties {group=QuietPeanutsStaple, parameterName=Scarf_SpecularOffset, sliderMax=3, sortPriority=31}
  }
  MaterialExpressionSetMaterialAttributes_0 {
   class=MaterialExpressionSetMaterialAttributes
   inputs [
    MaterialAttributes {from=MaterialExpressionGetMaterialAttributes_1, output=MaterialAttributes}
    Base Color {from=MaterialExpressionSwitch_0, output=""}
    Roughness {from=MaterialExpressionSaturate_0, output=""}
    Specular {from=MaterialExpressionSaturate_1, output=""}
   ]
   properties {attributeSetTypes [#1, #2, #3]}
  }
  MaterialExpressionSetMaterialAttributes_1 {
   class=MaterialExpressionSetMaterialAttributes
   inputs [
    MaterialAttributes {from=MaterialExpressionMaterialFunctionCall_8, output=MA}
    Base Color {from=MaterialExpressionNamedRerouteUsage_0, output=""}
    Emissive Color {from=MaterialExpressionNamedRerouteUsage_0, output=""}
   ]
   properties {attributeSetTypes [#1, #4]}
  }
  MaterialExpressionSubtract_0 {
   class=MaterialExpressionSubtract
   inputs [
    A {from=MaterialExpressionTextureSample_0, output=R}
    B {from=MaterialExpressionTextureSample_0, output=R}
   ]
  }
  MaterialExpressionSwitch_0 {
   class=MaterialExpressionSwitch
   inputs [
    SwitchValue {from=MaterialExpressionScalarParameter_0, output=""}
    Default {from="", output=""}
    None {from=MaterialExpressionDesaturation_3, output=""}
    None {from=MaterialExpressionLinearInterpolate_1, output=""}
    None {from=MaterialExpressionLinearInterpolate_2, output=""}
    None {from=MaterialExpressionLinearInterpolate_3, output=""}
   ]
   properties {inputs [
     {
      inputName=None
      input {
       expression=&self:MaterialExpressionDesaturation_3
       inputName=None
       mask=0
       maskA=0
       maskB=0
       maskG=0
       maskR=0
       outputIndex=0
      }
     }
     {
      inputName=None
      input {
       expression=&self:MaterialExpressionLinearInterpolate_1
       inputName=None
       mask=0
       maskA=0
       maskB=0
       maskG=0
       maskR=0
       outputIndex=0
      }
     }
     {
      inputName=None
      input {
       expression=&self:MaterialExpressionLinearInterpolate_2
       inputName=None
       mask=0
       maskA=0
       maskB=0
       maskG=0
       maskR=0
       outputIndex=0
      }
     }
     {
      inputName=None
      input {
       expression=&self:MaterialExpressionLinearInterpolate_3
       inputName=None
       mask=0
       maskA=0
       maskB=0
       maskG=0
       maskR=0
       outputIndex=0
      }
     }
    ]}
  }
  MaterialExpressionTextureSample_0 {
   class=MaterialExpressionTextureSample
   inputs [
    UVs {from=""}
    Tex {from=""}
    Apply View MipBias {from=""}
   ]
   properties {
    samplerType=SAMPLERTYPE_Masks
    texture=/BRCosmetics/Accessories/FORT_Backpacks/Backpack_M_MED_QuietPeanutsStaple/Textures/T_QuietPeanutsStaple_Backpack_FX
   }
  }
  MaterialExpressionVectorParameter_0 {
   class=MaterialExpressionVectorParameter
   properties {group=QuietPeanutsStaple, parameterName=Scarf_BlueColor, sortPriority=10, defaultValue=<rgba 0.098 0.3804 1 1>}
  }
  MaterialExpressionVectorParameter_1 {
   class=MaterialExpressionVectorParameter
   properties {group=QuietPeanutsStaple, parameterName=Scarf_RedColor, sortPriority=11, defaultValue=<rgba 1 0.1255 0.1882 1>}
  }
  MaterialExpressionVectorParameter_2 {
   class=MaterialExpressionVectorParameter
   properties {group=QuietPeanutsStaple, parameterName=Scarf_PurpleColor, sortPriority=12, defaultValue=<rgba 0.5216 0.2235 0.8549 1>}
  }
 }
 properties {
  bUseMaterialAttributes=true
  bUsedWithInstancedStaticMeshes=true
  bUsedWithMeshDeformer=true
  bUsedWithSkeletalMesh=true
  pixelDepthOffsetMode=PDOM_Legacy
  shadingModel=MSM_FromMaterialExpression
  usedShadingModels=Default Lit | Subsurface | Clear Coat | Subsurface Profile | Cloth
 }
}
