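# Physics asset for the QuietPeanutsStaple backpack cosmetic, giving its left and right ribbons collision bodies and joints for physical motion.
# A large capsule on master_shell anchors two chains, each a small base capsule followed by five dyn ribbon capsules.
# Base joints have free swing and twist, while downstream joints limit swing and nearly lock twist.
# Swing2 limits differ per side, widening toward the left tip and tightening toward the right tip.
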PhysicsAsset /BRCosmetics/Accessories/FORT_Backpacks/Backpack_M_MED_QuietPeanutsStaple/Meshes/M_MED_QuietPeanutsStaple_Pack_Physics
bodies {
 base_ribbon_00_l [{shapeType=Capsule, length=0.505, radius=0.505, shapeName=base_ribbon_00_l_capsule, rotation=<rot 0 -90 0>}]
 base_ribbon_00_r [{shapeType=Capsule, length=0.505, radius=0.505, shapeName=base_ribbon_00_r_capsule, rotation=<rot 0 -90 0>}]
 dyn_ribbon_00_l [{shapeType=Capsule, length=0.282, radius=1, shapeName=dyn_ribbon_00_l_capsule, center=<0.9877 0 0>, rotation=<rot 90 0 0>}]
 dyn_ribbon_00_r [{shapeType=Capsule, length=0.4547, radius=1, shapeName=dyn_ribbon_00_r_capsule, center=<1.178 0 0>, rotation=<rot 90 0 0>}]
 dyn_ribbon_01_l [{shapeType=Capsule, length=1.031, radius=1, shapeName=dyn_ribbon_01_l_capsule, center=<0.7739 0 0>, rotation=<rot 90 0 0>}]
 dyn_ribbon_01_r [{shapeType=Capsule, length=1.6076, radius=1, shapeName=dyn_ribbon_01_r_capsule, center=<1.7759 0 -0.2537>, rotation=<rot 90 0 0>}]
 dyn_ribbon_02_l [{shapeType=Capsule, length=2, radius=1, shapeName=dyn_ribbon_02_l_capsule, center=<2 0 0>, rotation=<rot 90 0 0>}]
 dyn_ribbon_02_r [{shapeType=Capsule, length=2, radius=1, shapeName=dyn_ribbon_02_r_capsule, center=<2 0 -0.2257>, rotation=<rot 90 0 0>}]
 dyn_ribbon_03_l [{shapeType=Capsule, length=2, radius=1, shapeName=dyn_ribbon_03_l_capsule, center=<2.3124 0 0>, rotation=<rot 90 0 0>}]
 dyn_ribbon_03_r [{shapeType=Capsule, length=2, radius=1, shapeName=dyn_ribbon_03_r_capsule, center=<2.3153 0 -0.2984>, rotation=<rot 90 0 0>}]
 dyn_ribbon_04_l [{shapeType=Capsule, length=2, radius=1, shapeName=dyn_ribbon_04_l_capsule, center=<2 0 0>, rotation=<rot 90 0 0>}]
 dyn_ribbon_04_r [{shapeType=Capsule, length=2, radius=1, shapeName=dyn_ribbon_04_r_capsule, center=<2 0 0>, rotation=<rot 90 0 0>}]
 master_shell [{
   shapeType=Capsule
   length=6.7223
   radius=12.0352
   shapeName=master_shell_capsule
   center=<-0.3352 8.8765 -8.7172>
   rotation=<rot 0 0 -32.5651>
  }]
}
constraints [
 {
  bone1Name=base_ribbon_00_l
  bone2Name=master_shell
  swing1LimitDegrees=45
  swing1Motion=Free
  swing2LimitDegrees=45
  swing2Motion=Free
  twistLimitDegrees=45
  twistMotion=Free
 }
 {bone1Name=dyn_ribbon_00_l, bone2Name=base_ribbon_00_l, swing1LimitDegrees=25, swing2LimitDegrees=20, twistLimitDegrees=1}
 {bone1Name=dyn_ribbon_01_l, bone2Name=dyn_ribbon_00_l, swing1LimitDegrees=35, swing2LimitDegrees=20, twistLimitDegrees=1}
 {bone1Name=dyn_ribbon_02_l, bone2Name=dyn_ribbon_01_l, swing1LimitDegrees=35, swing2LimitDegrees=20, twistLimitDegrees=1}
 {bone1Name=dyn_ribbon_03_l, bone2Name=dyn_ribbon_02_l, swing1LimitDegrees=35, swing2LimitDegrees=30, twistLimitDegrees=1}
 {bone1Name=dyn_ribbon_04_l, bone2Name=dyn_ribbon_03_l, swing1LimitDegrees=35, swing2LimitDegrees=30, twistLimitDegrees=1}
 {
  bone1Name=base_ribbon_00_r
  bone2Name=master_shell
  swing1LimitDegrees=45
  swing1Motion=Free
  swing2LimitDegrees=45
  swing2Motion=Free
  twistLimitDegrees=45
  twistMotion=Free
 }
 {bone1Name=dyn_ribbon_00_r, bone2Name=base_ribbon_00_r, swing1LimitDegrees=30, swing2LimitDegrees=25, twistLimitDegrees=1}
 {bone1Name=dyn_ribbon_01_r, bone2Name=dyn_ribbon_00_r, swing1LimitDegrees=35, swing2LimitDegrees=25, twistLimitDegrees=1}
 {bone1Name=dyn_ribbon_02_r, bone2Name=dyn_ribbon_01_r, swing1LimitDegrees=35, swing2LimitDegrees=20, twistLimitDegrees=1}
 {bone1Name=dyn_ribbon_03_r, bone2Name=dyn_ribbon_02_r, swing1LimitDegrees=35, swing2LimitDegrees=15, twistLimitDegrees=1}
 {bone1Name=dyn_ribbon_04_r, bone2Name=dyn_ribbon_03_r, swing1LimitDegrees=35, swing2LimitDegrees=10, twistLimitDegrees=1}
]
