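# Animation Blueprint for the QuietPeanutsStaple backpack, adding physics motion and fit corrections to the pack.
# In gameplay, offsets on the pack bone feed a RigidBody simulation with computed gravity override; the front end blends to a separate, fixed-gravity simulation.
# While skydiving, spine_05 yaw relative to spine_01 drives pack roll and shift.
# A final offset applies for female characters, detected when customization loads.
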AnimBlueprint /BRCosmetics/Accessories/FORT_Backpacks/Backpack_M_MED_QuietPeanutsStaple/Meshes/M_MED_QuietPeanutsStaple_Pack_AnimBP
target_skeleton=&/M_MED_QuietPeanutsStaple_Pack_Skeleton
graphs [
 AnimGraph {nodes [
   n1 = "Transform (Modify) Bone - Bone: pack"(ComponentPose=Local To Component(LocalPose=Input Pose()), Translation="0,0.75,1.5", Rotation="0,0,0") {
    translationMode=BMM_Additive
    translationSpace=BCS_BoneSpace
    alphaBoolBlend {blendOption=Linear}
    boneToModify {boneName=pack}
    translation {y=0.75, z=1.5}
   }
   n2 = "Transform (Modify) Bone - Bone: pack"(
    ComponentPose=@n1
    Translation=Get Backpack Translation() -> (Backpack Translation:Vector)
    Rotation=Get Backpack Rotate() -> (Backpack Rotate:Rotator)
   ) {
    rotationMode=BMM_Additive
    rotationSpace=BCS_BoneSpace
    translationMode=BMM_Additive
    translationSpace=BCS_BoneSpace
    alphaBoolBlend {blendOption=Linear}
    alphaScaleBiasClamp {bInterpResult=true}
    boneToModify {boneName=pack}
   }
   n3 = RigidBody!(ComponentPose=@n2, Alpha, OverrideWorldGravity:Vector=Get GravityOverride(), ExternalForce) -> (Pose) {
    bClampLinearTranslationLimitToRefPose=true
    bOverrideWorldGravity=true
    overridePhysicsAsset=&/M_MED_QuietPeanutsStaple_Pack_Physics
    simulationSpace=BaseBoneSpace
    alphaBoolBlend {blendOption=Linear}
    baseBoneRef {boneName=root}
    componentAppliedLinearAccClamp=<3000 3000 3000>
    componentLinearAccScale=<0.45 0.45 0.45>
    componentLinearVelScale=<0.3 0.2 0.2>
    simSpaceSettings {
     maxAngularAcceleration=1000
     maxAngularVelocity=3
     maxLinearAcceleration=700
     maxLinearVelocity=500
     velocityScaleZ=0
     worldAlpha=0.2
     externalLinearDragV=<0.2 0.2>
    }
   }
   n4 = Component To Local(ComponentPose=@n3)
   n5 = RigidBody(OverrideWorldGravity="0,0,-980") {
    bClampLinearTranslationLimitToRefPose=true
    bOverrideWorldGravity=true
    overridePhysicsAsset=&M_MED_QuietPeanutsStaple_Pack_Physics
    simulationSpace=BaseBoneSpace
    alphaBoolBlend {blendOption=Linear}
    baseBoneRef {boneName=root}
    componentAppliedLinearAccClamp=<50000 50000 50000>
    componentLinearAccScale=<0.6 0.6 0.6>
    componentLinearVelScale=<0.6 6 0.6>
    simSpaceSettings {
     maxAngularAcceleration=50000
     maxAngularVelocity=50000
     maxLinearAcceleration=50000
     maxLinearVelocity=50000
     worldAlpha=0.5
     externalLinearDragV=<0.2 0.2>
    }
   }
   n6 = Component To Local(ComponentPose=@n5)
   n7 = Get bIsInFrontEnd()
   n8 = "Transform (Modify) Bone - Bone: pack"(
    ComponentPose=Local To Component(LocalPose=Blend Poses by bool(BlendPose_0=@n6, BlendPose_1=@n4, bActiveValue=@n7) {blendType=Linear})
    Alpha="Get Is Female  Alpha"() -> ("Is Female  Alpha":double)
    Translation="0,0.75,0"
    Rotation="0,0,0"
   ) {
    translationMode=BMM_Additive
    translationSpace=BCS_BoneSpace
    alphaBoolBlend {blendOption=Linear}
    boneToModify {boneName=pack}
    translation {y=0.75}
   }
   Output Pose(Result=Component To Local(ComponentPose=@n8)) {layerGroup=None}
  ]}
 EventGraph {nodes [
   Event BlueprintUpdateAnimation()
   > n1 = Sequence()
   > n2 = Gravity Override State Function DEPRECATED(CharacterPartAnimInstance=Self-Reference(), GameTimeInSeconds=GetGameTimeInSeconds(), GravityOverrideParams=Get Gravity Override Params())
   > Set GravityOverride(GravityOverride=@n2)
   n3 = Sequence(execute=@n1.then_1)
   > n4 = Branch(Condition=Get bIsSkyDiving())
   n5 = GetOwnerSkeletalMeshComponent()
   n6 = BreakRotator(InRot=BreakTransform(InTransform=GetDeltaTransformFromRefPose(self=@n5, BoneName=spine_05, BaseName=spine_01)).Rotation)
   > Set Backpack Rotate(Backpack Rotate:Rotator=MakeRotator(Roll=MapRangeClamped(Value=@n6.Yaw, InRangeA=0, InRangeB=-30, OutRangeB=20))) -> (Output_Get:Rotator)
   Set Backpack Rotate(execute=@n4.else, Backpack Rotate="0, 0, 0")
   n7 = Branch(execute=@n3.then_1, Condition=Get bIsSkyDiving())
   > Set Backpack Translation(Backpack Translation:Vector=MakeVector(Y=MapRangeClamped(Value=@n6.Yaw, InRangeA=0, InRangeB=-30, OutRangeB=-1.9))) -> (Output_Get:Vector)
   Set Backpack Translation(execute=@n7.else, Backpack Translation="0, 0, 0")
   n8 = Event OnCharacterCustomizationLoaded()
   n9 = Get Array Item with EFort Custom Part Type(Array:Array of Custom Character Part Object References=@n8.CustomParts) -> (Output:Custom Character Part Object Reference)
   > n10 = Set Is Female(Is Female:bool="Equal (Enum)"(
     A:EFortCustomGender Enum=Get GenderPermitted!(self:Custom Character Part Object Reference=@n9) -> (GenderPermitted:EFortCustomGender Enum=Invalid)
     B:EFortCustomGender Enum=Female
    )) -> (Output_Get:bool)
   > "Set Is Female  Alpha"("Is Female  Alpha":double="To Float (Boolean)"(InBool=@n10)) -> (Output_Get:double)
   TryGetPawnOwner()
   GetVectorArrayAverage!(Vectors:Array of Vectors=Make Array("[0]":Vector="To Vector (Float)"!(InDouble:double=@n6.Yaw) -> (ReturnValue:Vector)) -> (Array:Array of Vectors)) -> (ReturnValue:Vector)
  ]}
]
variables [
 Backpack Rotate {type=Rotator}
 Backpack Translation {type=Vector}
 Gravity Override Params {type=GravityOverrideParamsStruct}
 GravityOverride {type=Vector}
 Is Female {type=boolean}
 "Is Female  Alpha" {type=number}
]
